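# Translucent unlit parent material that gives skinned meshes a tinted, cloak-like see-through look with cracked-glass glow.
# Center-scaled scene color is desaturated and tinted for the fill, while stepped Fresnel and parallax cracks build the emissive.
# Motion, from a debug sine (default) or collection Opacity, blends toward expire visibility by timestamp.
# Custom-depth comparison fakes backface culling, mobile uses flat colors, and the Glitch and Refraction reroutes are constants.
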Material /CleanSweepDiagram/Assets/Material/M_CleanSweepDiagram_Parent
graph {
 properties {
  allowTranslucentCustomDepthWrites=true
  bEnableResponsiveAA=true
  bUsedWithClothing=true
  bUsedWithMeshDeformer=true
  bUsedWithMorphTargets=true
  bUsedWithSkeletalMesh=true
  blendMode=BLEND_Translucent
  pixelDepthOffsetMode=PDOM_Legacy
  refractionMethod=RM_PixelNormalOffset
  shadingModel=MSM_Unlit
  usedShadingModels=Unlit
 }
 nodes [
  Add()
  n1 = ScalarParameter() {defaultValue=2, parameterName=FresnelInnerDarkenEXP}
  n2 = VertexNormalWS()
  n3 = Saturate(None=LinearInterpolate(Alpha=Fresnel(ExponentIn=@n1, Normal=@n2) {exponent=2}) {constA=1, constB=-1})
  n4 = NamedRerouteUsage()
  Add(A=@n3, B=@n4)
  n5 = MaterialFunctionCall(
   UVs=ScreenPosition().ViewportUV
   Texture Scale=LinearInterpolate(
    A=ScalarParameter() {defaultValue=1.5, parameterName=Refraction_Scale}
    B=ScalarParameter() {defaultValue=1, parameterName=Refraction_Scale_Glitch}
    Alpha=NamedRerouteUsage()
   ) {constA=1.4}
  ) {materialFunction=&ScaleUVsByCenter}
  n6 = SceneColor!(UVs=@n5.UVs)
  n7 = Max(A=@n6.RGB, B=ComponentMask(None=SceneTexture(UVs=@n5.UVs) {sceneTextureId=PPI_DiffuseColor}.Color) {b=true, g=true, r=true})
  n8 = LinearInterpolate(
   A=ScalarParameter() {defaultValue=2, parameterName=FresnelInnerDarken_FillColor_Contrast}
   B=ScalarParameter() {defaultValue=5, parameterName=FillColor_Contrast}
   Alpha=NamedRerouteUsage()
  ) {constB=0.1}
  n9 = Power(Base=Desaturation(None=Multiply(A=@n7, B=@n8) {constB=5}, Fraction=ScalarParameter() {defaultValue=0.5, parameterName=Desaturation}))
  n10 = VectorParameter() {parameterName=FillColor, defaultValue=<rgba 0.4062 1 0.5547 1>}
  n11 = Multiply(A=@n10.RGB)
  n12 = NamedRerouteUsage()
  n13 = PreSkinnedPosition()
  n14 = MaterialFunctionCall(Float3=Divide(A=VertexInterpolator(VS=@n13).PS, B=Constant() {r=200})) {materialFunction=&BreakOutFloat3Components}
  n15 = Multiply(
   A=LinearInterpolate(A=@n7, Alpha=Saturate(None=@n9))
   B=LinearInterpolate(
    A=LinearInterpolate(A=@n10.RGB, B=@n11, Alpha=@n12)
    B=VectorParameter() {parameterName=FillColor2, defaultValue=<rgba 0.3797 1 0.2901 1>}.RGB
    Alpha=@n14.B
   )
  ) {constA=1}
  Add(A=@n15)
  n16 = Multiply(A=Time(), B=Constant() {r=0.3}) {constB=0.3}
  n17 = Floor(None=Multiply(A=Frac(None=@n16)) {constB=10})
  Divide(A=@n17) {constB=10}
  EyeAdaptationInverse(LightValueInput=NamedRerouteUsage(), AlphaInput=ScalarParameter() {defaultValue=0.2, parameterName=TODBalance})
  LinearInterpolate()
  n18 = LinearInterpolate(
   A=LinearInterpolate(
    A=ScalarParameter() {defaultValue=0.8, parameterName=FillOpacityStationary}
    B=ScalarParameter() {defaultValue=1, parameterName=FillOpacityStationary_Glitch}
    Alpha=NamedRerouteUsage()
   )
   B=ScalarParameter() {parameterName=FillOpacityMoving}
   Alpha=NamedRerouteUsage()
  )
  LinearInterpolate(B=@n18, Alpha=NamedRerouteUsage())
  n19 = ScalarParameter() {defaultValue=0.995, parameterName=RefractionStationary}
  n20 = ScalarParameter() {defaultValue=0.98, parameterName=RefractionMoving}
  n21 = MaterialFunctionCall() {materialFunction=&DebugTimeSine}
  n22 = CollectionParameter() {collection=&/MCP_CleanSweepDiagram, parameterName=Opacity}
  n23 = ScalarParameter() {defaultValue=0.5, parameterName=ExpireTransitionLength}
  n24 = MaterialFunctionCall(
   TimestampA=ScalarParameter() {parameterName=ExpireTimeStampA}
   TransitionLength_In=@n23
   TimestampB=ScalarParameter() {parameterName=ExpireTimeStampB}
   TransitionLength_Out=@n23
   Transition In Only?=StaticBool()
  ) {materialFunction=&MF_Timestamp_InOut_01}
  n25 = NamedRerouteDeclaration(None=LinearInterpolate(
    A=StaticSwitchParameter(True=@n21.Result, False=@n22) {defaultValue=true, parameterName=DebugAnimation}
    B=ScalarParameter() {defaultValue=1.3, parameterName=ExpireVisibility}
    Alpha=@n24.SmoothStep
   )) {name=Motion, nodeColor=<rgba 1 0 0.6094 1>}
  n26 = ActorPositionWS()
  n27 = CameraPositionWS()
  n28 = Saturate(None=Divide(A=Distance(A=@n26, B=@n27), B=ScalarParameter() {defaultValue=800, parameterName=RefractionDistanceDivisor}))
  n29 = ScalarParameter() {parameterName=TimeStampA}
  n30 = ScalarParameter() {defaultValue=0.5, parameterName=TransitionLength}
  n31 = ScalarParameter() {defaultValue=10000000000, parameterName=TimeStampB}
  n32 = StaticBool()
  n33 = LinearInterpolate(
   A=ScalarParameter() {defaultValue=1.2, parameterName=RefractionTransition}
   B=LinearInterpolate(B=LinearInterpolate(A=@n19, B=@n20, Alpha=@n25), Alpha=OneMinus(None=@n28)) {constA=1}
   Alpha=NamedRerouteDeclaration(None=MaterialFunctionCall(TimestampA=@n29, TransitionLength_In=@n30, TimestampB=@n31, TransitionLength_Out=@n30, Transition In Only?=@n32) {materialFunction=&MF_Timestamp_InOut_01}.Linear) {name=LinearTimeStamp, nodeColor=<rgba 1 0.0938 0 1>}
  )
  n34 = Abs(None=Divide(A=Subtract(A=@n33)))
  n35 = ScalarParameter() {defaultValue=0.05, parameterName=DitherScale}
  n36 = ScalarParameter() {defaultValue=1, parameterName=Visibility}
  n37 = LinearInterpolate(
   A=Add(A=@n34)
   B=Subtract(B=@n34)
   Alpha=Multiply(
    A=MaterialFunctionCall() {materialFunction=&DitherTemporalAA}.Result
    B=Multiply(A=@n35, B=@n36)
   )
  ) {constA=0.9, constB=1.1}
  LinearInterpolate(B=ShadingPathSwitch(Default=Constant() {r=1}, Deferred=Multiply(A=@n33, B=@n37)), Alpha=NamedRerouteUsage()) {constA=1}
  n38 = ScalarParameter() {defaultValue=0.5, parameterName=Opacity_LowQuality}
  n39 = StaticSwitchParameter(True=@n38, False=ScalarParameter() {parameterName=Opacity}) {parameterName=LowQualityPreview}
  n40 = Constant() {r=1}
  n41 = SceneTexture() {sceneTextureId=PPI_CustomDepth}
  n42 = Constant()
  n43 = If(
   A=PixelDepth().R
   B=Add(A=ComponentMask(None=@n41.Color) {r=true}, B=ScalarParameter() {defaultValue=0.1, parameterName=cdoffset}) {constB=0.1}
   "A > B"=@n42
   "A == B"=@n42
   "A < B"=Constant() {r=1}
  )
  Multiply(
   A=ShadingPathSwitch(Default=@n39, Deferred=@n39, Mobile=@n38) {desc=Migrated from FeatureLevelSwitch}
   B=StaticSwitchParameter(True=QualitySwitch(Default=@n40, High=@n43, Epic=@n43), False=@n40) {defaultValue=true, parameterName=EnableBackFaceCulling}
  )
  n44 = Multiply(A=Time(), B=Constant() {r=0.15}) {constB=0.3}
  n45 = Frac(None=@n44)
  n46 = Constant() {r=1}
  n47 = Frac(None=Multiply(A=Sine(None=@n45), B=@n46))
  n48 = Sine(None=Multiply(A=@n44) {constB=2})
  n49 = Frac(None=Power(Base=Multiply(A=@n46, B=@n48)))
  n50 = ScreenPosition()
  n51 = Constant2Vector() {g=0.5, r=2}
  n52 = TextureSampleParameter2D(UVs=Panner(Coordinate=Multiply(A=@n50.ViewportUV, B=@n51), Time=@n17) {speedY=0.1}) {
   parameterName=GlitchTexture
   samplerType=SAMPLERTYPE_Masks
   texture=/Game/Characters/Player/Female/Medium/Bodies/F_MED_BlueBadass/Skins/NeonLines/F_MED_BlueBadass_NeonLines_-_GlitchNoise_v1
  }
  n53 = Multiply(A=SphereMask(A=Max(A=@n47, B=@n49), B=Constant() {r=0.5}) {attenuationRadius=0.3}, B=Power(Base=@n52.RGB) {constExponent=1})
  Multiply(A=Multiply(A=@n53, B=Constant3Vector() {constant=<rgba 0.3876 1 0.2843 1>})) {constB=0.2}
  n54 = VectorParameter() {parameterName=BaseFresnelColor, defaultValue=<rgba 0 1 0.5013 1>}
  n55 = SphereMask(A=NamedRerouteUsage(), B=VectorParameter() {defaultValue=<rgba 0 4 154 1>}.RGB) {attenuationRadius=18, hardnessPercent=50}
  Multiply(A=@n54.RGB, B=NamedRerouteDeclaration(None=@n55) {name=HeadMask, nodeColor=<rgba 0.5703 1 0 1>})
  NamedRerouteDeclaration(None=Constant()) {name=Glitch, nodeColor=<rgba 0.2656 0 1 1>}
  n56 = PreSkinnedPosition()
  n57 = Divide(
   A=NamedRerouteDeclaration(None=VertexInterpolator(VS=@n56).PS) {name=PreskinnedLocalPosition, nodeColor=<rgba 0.7344 0 1 1>}
   B=Constant() {r=200}
  )
  n58 = ScalarParameter() {defaultValue=1, parameterName=CracksTiling}
  n59 = BumpOffset(
   Coordinate=ComponentMask(None=Multiply(A=@n57, B=@n58) {constB=-1}) {b=true, r=true}
   Height=Constant() {r=1}
   HeightRatioInput=ScalarParameter() {defaultValue=-0.3, parameterName=CracksHeightRatio}
  )
  n60 = Power(Base=TextureSample(UVs=NamedRerouteDeclaration(None=@n59) {name=Crack UV, nodeColor=<rgba 0 0.0625 1 1>}) {texture=/Game/Effects/Fort_Effects/Textures/Tile/T_Glass_Cracks_M}.G) {constExponent=1.5}
  n61 = ScalarParameter() {defaultValue=4, parameterName=CracksFadeInOpacity}
  n62 = VectorParameter() {parameterName=EmissiveColor, defaultValue=<rgba 0.6282 0.9926 5 1>}
  n63 = VectorParameter() {parameterName=EmissiveColor_FresnelBlend, defaultValue=<rgba 0.6282 0.9926 5 1>}
  n64 = LinearInterpolate(
   A=ScalarParameter() {defaultValue=3, parameterName=FresnelExponentStationary}
   B=ScalarParameter() {defaultValue=0.4, parameterName=FresnelExponentMoving}
   Alpha=NamedRerouteUsage()
  )
  n65 = Saturate(None=Add(A=NamedRerouteUsage()) {constB=0.12})
  n66 = Multiply(
   A=TextureSample(UVs=NamedRerouteUsage()) {samplerType=SAMPLERTYPE_Normal, texture=/Game/Effects/Fort_Effects/Textures/Tile/T_glass_cracks_sharp_N}.RGB
   B=VectorParameter() {parameterName=DetailNormalIntensity, defaultValue=<rgba 0.5 0.5 1 0>}.RGB
  )
  n67 = ScalarParameter() {defaultValue=2, parameterName=QuantizeSteps}
  n68 = Multiply(
   A=MaterialFunctionCall(Exponent=Multiply(A=@n64, B=@n65), Vertex / Pixel Normal=Transform(None=@n66)) {materialFunction=&MF_Fresnel_FX}.Fresnel Inverted
   B=@n67
  )
  n69 = Add(A=Multiply(A=Multiply(A=@n60, B=@n61), B=@n62.RGB), B=Multiply(A=LinearInterpolate(A=@n63.RGB, B=@n62.RGB, Alpha=@n68), B=@n68))
  n70 = NamedRerouteUsage()
  n71 = OneMinus(None=@n70)
  n72 = LinearInterpolate(
   A=ScalarParameter() {parameterName=CracksOpacityStationary}
   B=ScalarParameter() {defaultValue=1, parameterName=CracksOpacityMoving}
   Alpha=NamedRerouteUsage()
  )
  n73 = Multiply(A=VectorParameter() {parameterName=CracksColor, defaultValue=<rgba 0.0023 0.1 0 1>}.RGB, B=Multiply(A=@n60, B=@n72))
  n74 = DotProduct(A=CameraVectorWS(), B=VertexNormalWS())
  n75 = NamedRerouteDeclaration(None=OneMinus(None=Saturate(None=@n74))) {name=FresnelBase, nodeColor=<rgba 0.8047 1 0 1>}
  n76 = ScalarParameter() {defaultValue=4, parameterName=BaseFresnelExponent}
  n77 = ScalarParameter() {parameterName=BaseFresnelStationary}
  n78 = ScalarParameter() {defaultValue=1, parameterName=BaseFresnelMoving}
  n79 = NamedRerouteUsage()
  n80 = SphereMask(
   A=TransformPosition(None=WorldPosition().XYZ) {transformSourceType=TRANSFORMPOSSOURCE_World}
   B=Constant3Vector() {constant=<rgba 0 2 150 0>}
  ) {attenuationRadius=15}
  n81 = Multiply(
   A=Multiply(A=@n54.RGB, B=Power(Base=@n75, Exp=@n76) {constExponent=8})
   B=LinearInterpolate(A=LinearInterpolate(A=@n77, B=@n78, Alpha=@n79), Alpha=@n80) {constB=0}
  )
  n82 = NamedRerouteUsage()
  n83 = Power(
   Base=DotProduct(A=Constant3Vector() {constant=<rgba 0.2722 0.6741 0.0537 1>}, B=SceneColor().RGB)
   Exp=ScalarParameter() {defaultValue=2, group=LuminanceCorrection, parameterName=LuminanceCorrectionPower}
  )
  n84 = LinearInterpolate(
   A=ScalarParameter() {defaultValue=8, group=LuminanceCorrection, parameterName=LuminanceCorrectionIntensityStationary}
   B=ScalarParameter() {defaultValue=4, group=LuminanceCorrection, parameterName=LuminanceCorrectionIntensityMoving}
   Alpha=NamedRerouteUsage()
  )
  n85 = ScalarParameter() {defaultValue=1, parameterName=OverallOpacity}
  n86 = Multiply(A=Add(A=Multiply(A=@n83, B=@n84) {constB=4}), B=@n85)
  n87 = Multiply(
   A=Add(A=Multiply(A=@n73, B=@n72), B=Add(A=@n81, B=@n82) {constB=0})
   B=ShadingPathSwitch(
    Default=StaticSwitchParameter(True=@n86, False=@n85) {defaultValue=true, group=LuminanceCorrection, parameterName=UseLuminanceCorrection, sortPriority=0}
    Mobile=@n85
   ) {desc=Migrated from FeatureLevelSwitch}
  )
  n88 = Add(A=Multiply(A=@n24.SmoothStep, B=ScalarParameter() {defaultValue=4, parameterName=ExpireGlowIntensity}) {constB=4})
  n89 = Divide(A=Floor(None=@n68), B=@n67)
  n90 = Saturate(None=Add(A=OneMinus(None=@n89), B=ConstantBiasScale(None=@n70) {bias=-0.5, scale=2}))
  n91 = NamedRerouteUsage()
  NamedRerouteDeclaration(None=LinearInterpolate(A=Multiply(A=@n69, B=@n71), B=Multiply(A=@n87, B=@n88), Alpha=LinearInterpolate(A=@n90, Alpha=@n91))) {name=Emissive, nodeColor=<rgba 0.875 0 1 1>}
  NamedRerouteDeclaration(None=@n80) {name=HeadMask_1, nodeColor=<rgba 1 0 0.7734 1>}
  n92 = VertexNormalWS()
  n93 = ScalarParameter() {defaultValue=2, parameterName=QuantizeSteps}
  n94 = Floor(None=Multiply(A=@n92, B=@n93))
  NamedRerouteDeclaration(None=ShadingPathSwitch(Default=@n92, Deferred=Divide(A=@n94, B=@n93))) {name=QuantizeNormal, nodeColor=<rgba 0 0.25 1 1>}
  NamedRerouteDeclaration(None=LinearInterpolate(B=@n3, Alpha=@n4) {constA=1}) {name=FresnelDarkening, nodeColor=<rgba 0 0.4375 1 1>}
  NamedRerouteDeclaration(None=Constant() {r=1}) {name=Refraction, nodeColor=<rgba 0 0.1328 1 1>}
  n95 = Subtract(A=ObjectPositionWS(), B=WorldPosition().XYZ)
  n96 = ScalarParameter() {parameterName=HT_CrunchVerts}
  NamedRerouteDeclaration(None=Add(
    A=MaterialFunctionCall!(WPO) {materialFunction=/Game/Packages/Fortress_SharedMaterials/Base_Material_Functions/Fortnite_HideElementsWithVertexColor}.Result
    B=Multiply(A=@n95, B=@n96)
   )) {name=WPO, nodeColor=<rgba 0.0781 1 0 1>}
  n97 = LinearInterpolate(
   A=VectorParameter() {parameterName=FillColor_LowQaulity, defaultValue=<rgba 0.039 0.2031 0.0252 1>}.RGB
   B=VectorParameter() {parameterName=FillColor_LowQaulity2, defaultValue=<rgba 0 0 0 1>}.RGB
   Alpha=@n14.B
  )
  n98 = LinearInterpolate(A=Multiply(A=@n97) {constB=0.5}, B=@n97, Alpha=NamedRerouteUsage())
  n99 = StaticSwitchParameter(
   True=MaterialFunctionCall!(Color=@n98) {materialFunction=/Game/Packages/Fortress_SharedMaterials/Anime_Master_Material/MF_TODColorCurve}.Result
   False=LinearInterpolate(A=@n6.RGB, B=@n15, Alpha=@n18)
  ) {parameterName=LowQualityPreview}
  NamedRerouteDeclaration(None=ShadingPathSwitch(Default=@n99, Deferred=@n99, Mobile=@n98) {desc=Migrated from FeatureLevelSwitch}) {name=BaseTint, nodeColor=<rgba 0 0.9297 1 1>}
  NamedRerouteUsage()
  NamedRerouteUsage()
  NamedRerouteUsage()
  OneMinus(None=@n3)
  OneMinus(None=@n55)
  Saturate(None=@n53)
  Subtract(A=@n3, B=@n4)
 ]
}
